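# Translucent, unlit, two-sided test material for a glowing tech-screen edge effect on the female Snow Board character cosmetic.
# One mask texture supplies edge color, a parallax fiber layer and a sine-pulsed tech glow, with a fiber-distorted panning sample varying glow contrast.
# Emissive gets an HDR-from-LDR remap outside mobile.
# Scanlines form opacity, softened by a noise-varied depth fade that mobile and low-quality paths skip.
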Material /BRCosmetics/Characters/Player/Female/Medium/Bodies/F_MED_Snow_Board/Particles/M_Snow_Board_Edge_Test_03
graph {
 properties {
  bUsedWithMeshDeformer=true
  bUsedWithSkeletalMesh=true
  blendMode=BLEND_Translucent
  pixelDepthOffsetMode=PDOM_Legacy
  shadingModel=MSM_Unlit
  translucencyLightingMode=TLM_VolumetricPerVertexNonDirectional
  twoSided=true
  usedShadingModels=Unlit
 }
 nodes [
  n1 = LinearInterpolate(
   A=ScalarParameter() {defaultValue=5, parameterName=TechBrightness}
   B=Constant() {r=0.35}
   Alpha=MaterialFunctionCall(Speed=ScalarParameter() {defaultValue=0.075, parameterName=SinSpeed}) {materialFunction=&DebugTimeSine}.Result
  )
  Clamp(None=@n1)
  n2 = VectorParameter() {parameterName=TechColor_Dim, defaultValue=<rgba 0.1049 0.2166 0.4757 1>}
  n3 = VectorParameter() {parameterName=TechColorBright, defaultValue=<rgba 0.2569 1.4973 3.125 0>}
  n4 = ScalarParameter() {defaultValue=0.51, parameterName=Pow}
  n5 = TextureCoordinate() {uTiling=0.5, vTiling=0.7}
  n6 = TextureCoordinate()
  n7 = ScalarParameter() {defaultValue=-0.08, parameterName=HEight}
  n8 = Multiply(
   A=TextureSample(UVs=BumpOffset(Coordinate=@n6, Height=@n7) {heightRatio=0.3}) {texture=&/T_Snow_Board_Screens_Mask_D3}.B
   B=ScalarParameter() {defaultValue=1.2, parameterName=FiberScale}
  )
  n9 = ScalarParameter() {defaultValue=0.03, parameterName=UVdistort}
  n10 = TextureSample(UVs=Add(A=Panner(Coordinate=@n5) {speedX=0.025}, B=Multiply(A=@n8, B=@n9))) {texture=&T_Snow_Board_Screens_Mask_D3}
  n11 = Multiply(
   A=Power(
    Base=TextureSample() {texture=&T_Snow_Board_Screens_Mask_D3}.R
    Exp=LinearInterpolate(A=@n4, Alpha=@n10.B) {constB=6}
   )
   B=@n1
  )
  n12 = Add(
   A=LinearInterpolate(
    A=VectorParameter() {parameterName=BackgroundColor, defaultValue=<rgba 0.0339 0.0879 0.1285 0>}.RGB
    B=VectorParameter() {parameterName=EdgeColor, defaultValue=<rgba 0 3.2814 6 0>}.RGB
    Alpha=TextureSample() {texture=&T_Snow_Board_Screens_Mask_D3}.G
   )
   B=Add(A=LinearInterpolate(A=@n2.RGB, B=@n3.RGB, Alpha=@n11), B=@n8)
  )
  n13 = Clamp(None=@n12)
  n14 = ScalarParameter() {defaultValue=0.8, parameterName=DesatLuminance}
  MaterialFunctionCall(Emissive=ShadingPathSwitch(
    Default=MaterialFunctionCall(LDR Input=@n13, HDR Tint=@n12, Desaturate Luminance=@n14) {materialFunction=&DeriveHDRfromLDR}.RemappedRange
    Mobile=@n12
   ) {desc=Migrated from FeatureLevelSwitch}) {materialFunction=&MF_FX_EmissiveAdjustments}
  Multiply(A=@n3.RGB, B=@n11)
  n15 = MaterialFunctionCall!(
   Number of Scanlines=ScalarParameter() {defaultValue=90, parameterName=Scanline Amount}
   Scanline Speed
   "Scanlines Thickness (Bias)"=ScalarParameter() {defaultValue=-0.2, parameterName=ScanlineSize}
   Scanlines Dark Value=ScalarParameter() {defaultValue=0.7, parameterName=Scanlines Dark}
   Scanlines Light Value=ScalarParameter() {defaultValue=0.85, parameterName=Scanlines Light}
   Block Opacity=Constant()
   UVs
   Time
   Block Repeat
  ) {materialFunction=/Game/Effects/Fort_Effects/Effects/Atlas/Materials/MF_Atlas_Scanlines_01}
  n16 = Add(A=@n15.Result, B=Multiply(A=@n15.Secondary Lines, B=ScalarParameter() {defaultValue=0.5, parameterName=SeconddaryLine}))
  n17 = Multiply(A=ScalarParameter() {defaultValue=0.8, parameterName=Opacity}, B=@n16)
  n18 = Panner(Coordinate=Multiply(A=TextureCoordinate() {uTiling=0.2, vTiling=0.2}, B=ScalarParameter() {defaultValue=0.5, parameterName=ScaleDepth})) {speedX=0.005}
  n19 = DepthFade(
   Opacity=@n17
   FadeDistance=LinearInterpolate(
    A=ScalarParameter() {defaultValue=4, parameterName=Depth0}
    B=ScalarParameter() {defaultValue=30, parameterName=Depth1}
    Alpha=TextureSample(UVs=@n18) {texture=/Game/Global/Textures/Default/Noise/T_TileNoise_02}.G
   )
  )
  QualitySwitch(Default=@n19, Low=@n17)
  ShadingPathSwitch(Default=@n19, Mobile=Multiply(A=ScalarParameter() {defaultValue=0.35, parameterName=OpacityMobile}, B=@n16)) {desc=Migrated from FeatureLevelSwitch}
 ]
}
